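# Translucent, unlit, two-sided material that gives quest props a ghostly hologram look, kept in a Test folder.
# Two opposite-panning world-position scanline samples modulate opacity, shaped by Fresnel terms and camera depth fade.
# Emissive blends two time-of-day color sets by Fresnel, scaled by rim and scanline intensity.
# The Atlas hologram wiggle function supplies world position offset distortion.
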Material /Game/Missions/Quests/ReactiveQuests/Test/M_QuestPropGhostly
graph {
 properties {
  bUsedWithNanite=true
  blendMode=BLEND_Translucent
  pixelDepthOffsetMode=PDOM_Legacy
  shadingModel=MSM_Unlit
  twoSided=true
  usedShadingModels=Unlit
 }
 nodes [
  n1 = Fresnel(ExponentIn=ScalarParameter() {defaultValue=2, desc="0 = Opaque, Default = 4, 20 = way more transparent", parameterName=Fresnel})
  n2 = Multiply(A=@n1, B=ScalarParameter() {defaultValue=0.75, parameterName=Fresnel Strength}) {constB=0.75}
  n3 = Multiply(
   A=StaticComponentMaskParameter(None=WorldPosition().XYZ) {defaultB=true, defaultR=true, parameterName=Lines Mask Direction}
   B=ScalarParameter() {defaultValue=0.025, parameterName=ScanlineScale}
  )
  n4 = TextureSample(UVs=Panner(Coordinate=@n3) {bFractionalPart=true, speedY=0.8}) {texture=/Game/Global/Textures/Default/Masks_and_Gradients/Scanlines_Offset}
  n5 = TextureSample(UVs=Panner(Coordinate=Multiply(A=@n3) {constB=0.5}) {bFractionalPart=true, speedY=-0.4}) {texture=&Scanlines_Offset}
  n6 = Multiply(
   A=OneMinus(None=@n1)
   B=LinearInterpolate(
    A=ScalarParameter() {defaultValue=0.05, parameterName=BaseOpacity}
    B=ScalarParameter() {defaultValue=0.55, parameterName=Line Intensity}
    Alpha=Add(A=@n4.G, B=@n5.G)
   )
  )
  n7 = Fresnel(ExponentIn=ScalarParameter() {defaultValue=2, desc="0 = Opaque, Default = 4, 20 = way more transparent", parameterName=Fresnel})
  n8 = MaterialFunctionCall(Fade Length=Constant() {r=-1000}, Fade Offset=Constant() {r=20000}) {materialFunction=&CameraDepthFade}
  Add(A=@n2, B=Multiply(A=@n6, B=Multiply(A=@n7, B=@n8.Result)))
  MaterialFunctionCall!(
   WPO Intensity A=ScalarParameter() {defaultValue=0.4, parameterName=Holo Distortion Intensity Base}
   WPO Intensity B=ScalarParameter() {defaultValue=0.2, parameterName=Holo Distortion Intensity Distorted}
   Time
   Bias=Constant() {r=0.2}
  ) {materialFunction=/Game/Effects/Fort_Effects/Effects/Atlas/Materials/MF_Atlas_HologramWiggle_01}
  n9 = VectorParameter() {parameterName=Color_MorningEvening, defaultValue=<rgba 0.28 0.2403 0.14 1>}
  n10 = VectorParameter() {parameterName=Color_Day, defaultValue=<rgba 0.55 0.4721 0.275 1>}
  n11 = VectorParameter() {parameterName=Color_Night, defaultValue=<rgba 0.205 0.176 0.1025 1>}
  n12 = VectorParameter() {parameterName=Color_2_MorningEvening, defaultValue=<rgba 0 0.3655 0.9028 1>}
  n13 = VectorParameter() {parameterName=Color_2_Day, defaultValue=<rgba 0 0.2125 1 1>}
  n14 = VectorParameter() {parameterName=Color_2_Night, defaultValue=<rgba 0 0.2465 0.5903 1>}
  n15 = ScalarParameter() {defaultValue=2, desc="0 = Opaque, Default = 4, 20 = way more transparent", parameterName=Color Shift Fresnel}
  Multiply(
   A=LinearInterpolate(
    A=MaterialFunctionCall(Color_Morning=@n9.RGB, Color_Day=@n10.RGB, Color_Evening=@n9.RGB, Color_Night=@n11.RGB) {materialFunction=&Color_Time_of_Day}.Result
    B=MaterialFunctionCall(Color_Morning=@n12.RGB, Color_Day=@n13.RGB, Color_Evening=@n12.RGB, Color_Night=@n14.RGB) {materialFunction=&Color_Time_of_Day}.Result
    Alpha=Fresnel(ExponentIn=@n15)
   )
   B=Add(A=@n2, B=@n6)
  )
 ]
}
